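; Gameplay ability that scales its owner's base health and shield to the playlist's max team size.
; On activation it waits for playlist data, then stores the team size, limiting it to 1–4.
; In the over-four branch, it applies any valid health and shield GameplayEffects, passing team size as level, then ends.
; The under-one branch only sets the value, and in-range sizes have no branch shown.
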(event Ability|EventActivateAbility
  (bind _max_team_size (Variables|InternalData|GetMax%20Team%20Size))
  (bind _ge_basehealth (Variables|Default|GetGE_BaseHealth))
  (bind _ge_baseshield (Variables|Default|GetGE_BaseShield))
  (bind (_gamestate _playlist) (Class|FortAsyncActionCurrentPlaylistReady|PlaylistDataReadyAsync)
    (:OnReady
      (bind _output_get (Variables|InternalData|SetMax%20Team%20Size (Variables|/Script/FortniteGame.FortPlaylist|GetMaxTeamSize _playlist)))
      (if (< _output_get 1)
        (Variables|InternalData|SetMax%20Team%20Size 1)
        (elif (> (Variables|InternalData|GetMax%20Team%20Size) 4)
          (Variables|InternalData|SetMax%20Team%20Size 4)
          (if (Utilities|IsValidClass _ge_basehealth)
            (Ability|ApplyGameplayEffectToOwner _ge_basehealth _max_team_size)
            (else
              _))
          (if (Utilities|IsValidClass _ge_baseshield)
            (Ability|ApplyGameplayEffectToOwner _ge_baseshield _max_team_size)
            (else
              _))
          (Ability|EndAbility))))))

(event Ability|EventOnEndAbility
  (params (bWasCancelled :type Boolean)))

(event Ability|EventFailedToActivatePassiveAbility
  (params (ActorInfo :type Gameplay%20Ability%20Actor%20Info%20Structure)))
